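# Material function that adds animated, music-reactive party lights to the Armored Battle Bus emissive, behind a static switch.
# It projects blurred and sharp panning light textures from pre-skinned direction around an object pivot, layering a hue-shifted second set and a sphere-masked hot core.
# Music amplitude or a sine drives pulsing, also output as MusicPulsing, and timestamps fade the lights.
# Switches add cheap Fresnel shading or masked pulsing emissive.
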MaterialFunction /Game/Packages/Fortress_SharedMaterials/PartyBus_TempLocation/MF_ArmoredBattleBus_PartyLights
graph {nodes [
  n1 = GetMaterialAttributes(None=FunctionInput() {inputType=FunctionInput_MaterialAttributes}) {attributeGetTypes [#1, #2]}
  n2 = NamedRerouteUsage()
  n3 = VertexInterpolator(VS=PreSkinnedNormal())
  n4 = NamedRerouteUsage()
  n5 = Multiply(
   A=MaterialFunctionCall(
    Brightness=ScalarParameter() {defaultValue=1, group=PartyLights, parameterName=SimpleShadingFresnelBrightness}
    Exponent=ScalarParameter() {defaultValue=1, group=PartyLights, parameterName=SimpleShadingFresnelExponent}
    Camera Vector=NamedRerouteUsage()
    Vertex / Pixel Normal=Multiply(A=@n3.PS) {constB=-1}
   ) {materialFunction=&MF_Fresnel_FX}.Fresnel
   B=SphereMask(
    A=MaterialFunctionCall(Float3=@n4) {materialFunction=&BreakOutFloat3Components}.B
    B=ScalarParameter() {group=PartyLights, parameterName=SimpleShadingZMaskHeight}
    Radius=ScalarParameter() {defaultValue=1, group=PartyLights, parameterName=SimpleShadingZMaskRadius}
    Hardness=ScalarParameter() {defaultValue=1, group=PartyLights, parameterName=SimpleShadingZMaskHardness}
   )
  )
  n6 = Multiply(A=VectorParameter() {group=PartyLights, parameterName=CoreColorScale, defaultValue=<rgba 25 25 25 0>}.RGB, B=NamedRerouteUsage())
  n7 = NamedRerouteDeclaration(None=VertexInterpolator(VS=PreSkinnedPosition()).PS) {name=PreSkinnedLocalPosition, nodeColor=<rgba 0 0.9531 1 1>}
  n8 = NamedRerouteDeclaration(None=VectorParameter() {group=PartyLights, parameterName=ObjectPivot, defaultValue=<rgba -127.6242 0 360 0>}.RGB) {name=ObjectPivot, nodeColor=<rgba 0.5938 1 0 1>}
  n9 = NamedRerouteDeclaration(None=Normalize(VectorInput=Subtract(A=@n7, B=@n8))) {name=UnitVectorFromCenterToPixel, nodeColor=<rgba 0.2169 1 0 1>}
  n10 = VertexInterpolator(VS=PreSkinnedNormal())
  n11 = MaterialFunctionCall(Float3=Abs(None=@n10.PS)) {materialFunction=&BreakOutFloat3Components}
  n12 = LinearInterpolate(
   A=Multiply(A=ComponentMask(None=@n9) {b=true, g=true}) {constB=-1}
   B=ComponentMask(None=@n9) {b=true, r=true}
   Alpha=Floor(None=Add(A=@n11.G) {constB=0.001})
  )
  n13 = ScalarParameter() {defaultValue=2, group=PartyLights, parameterName=Lights_U_Scale_B}
  n14 = ScalarParameter() {defaultValue=2, group=PartyLights, parameterName=Lights_V_Scale_B}
  n15 = ScalarParameter() {defaultValue=-0.781, group=PartyLights, parameterName=Lights_U_PanSpeed_B}
  n16 = NamedRerouteUsage()
  n17 = Panner(
   Coordinate=Multiply(A=Multiply(A=@n12) {constB=-1}, B=AppendVector(A=@n13, B=@n14))
   Speed=AppendVector(A=Multiply(A=@n15, B=@n16), B=ScalarParameter() {defaultValue=0.0485, group=PartyLights, parameterName=Lights_V_PanSpeed_B})
  ) {bFractionalPart=true}
  n18 = TextureSample!(UVs=@n17, Tex, Bias=ScalarParameter() {defaultValue=7, group=PartyLights, parameterName=BlurredLightsMIPBias}, Apply View MipBias) {
   automaticViewMipBias=false
   mipValueMode=TMVM_MipBias
   samplerType=SAMPLERTYPE_LinearColor
   texture=&/T_ArmoredBattleBus_PartyLights
  }
  n19 = TextureSample(UVs=@n17, Bias=ScalarParameter() {group=PartyLights, parameterName=UnblurredLightsMIPBias}) {
   mipValueMode=TMVM_MipBias
   samplerType=SAMPLERTYPE_LinearColor
   texture=&T_ArmoredBattleBus_PartyLights
  }
  n20 = MaterialFunctionCall(
   Hue Shift Percentage=ScalarParameter() {defaultValue=0.15, group=PartyLights, parameterName=Hue Shift Percentage}
   Texture=LinearInterpolate(A=@n18.RGB, B=@n19.RGB, Alpha=NamedRerouteUsage())
  ) {materialFunction=&HueShift}
  n21 = VectorParameter() {group=PartyLights, parameterName=LightsColorScaleB, defaultValue=<rgba 0.3281 0.3281 0.3281 0>}
  n22 = ScalarParameter() {defaultValue=1.5, group=PartyLights, parameterName=Lights_U_Scale}
  n23 = ScalarParameter() {defaultValue=1.5, group=PartyLights, parameterName=Lights_V_Scale}
  n24 = ScalarParameter() {defaultValue=0.44, group=PartyLights, parameterName=Lights_U_PanSpeed}
  n25 = NamedRerouteUsage()
  n26 = Panner(
   Coordinate=Multiply(A=@n12, B=AppendVector(A=@n22, B=@n23))
   Speed=AppendVector(A=Multiply(A=@n24, B=@n25), B=ScalarParameter() {defaultValue=-0.062, group=PartyLights, parameterName=Lights_V_PanSpeed})
  ) {bFractionalPart=true}
  n27 = TextureSample(UVs=@n26, Bias=ScalarParameter() {defaultValue=10, group=PartyLights, parameterName=BlurredLightsMIPBias}) {
   automaticViewMipBias=false
   mipValueMode=TMVM_MipBias
   samplerType=SAMPLERTYPE_LinearColor
   texture=&T_ArmoredBattleBus_PartyLights
  }
  n28 = TextureSample(UVs=@n26, Bias=ScalarParameter() {group=PartyLights, parameterName=UnblurredLightsMIPBias}) {
   mipValueMode=TMVM_MipBias
   samplerType=SAMPLERTYPE_LinearColor
   texture=&T_ArmoredBattleBus_PartyLights
  }
  n29 = Multiply(
   A=LinearInterpolate(A=@n27.RGB, B=@n28.RGB, Alpha=NamedRerouteUsage())
   B=VectorParameter() {group=PartyLights, parameterName=LightsColorScale, defaultValue=<rgba 1 1 1 0>}.RGB
  )
  n30 = TextureSample(UVs=@n26, Bias=ScalarParameter() {defaultValue=8, group=PartyLights, parameterName=CoreLightsMIPBias}) {
   mipValueMode=TMVM_MipBias
   samplerType=SAMPLERTYPE_LinearColor
   texture=&T_ArmoredBattleBus_PartyLights
  }
  n31 = ScalarParameter() {defaultValue=300, group=PartyLights, parameterName=CoreMaskRadiusSmall}
  n32 = ScalarParameter() {defaultValue=425, group=PartyLights, parameterName=CoreMaskRadius}
  n33 = Sine(None=Multiply(A=Time(), B=ScalarParameter() {defaultValue=0.6, parameterName=SinewaveSpeed})) {period=0.6}
  n34 = Multiply(
   A=StaticSwitchParameter(
    True=ConstantBiasScale(None=@n33)
    False=CollectionParameter() {collection=/Game/Athena/Events/AudioAnalysis/MPC_Audio_Reactive_Parameters, parameterName=Amplitude_Music}
   ) {group=PartyLights, parameterName=UseSinewaveforAmplitude}
   B=ScalarParameter() {defaultValue=1.5, group=PartyLights, parameterName=AudioFFTScale}
  )
  n35 = Multiply(
   A=NamedRerouteDeclaration(None=FunctionOutput(None=@n34) {outputName=MusicPulsing, sortPriority=1}) {name=Pulsing, nodeColor=<rgba 1 0.5156 0 1>}
   B=ScalarParameter() {group=PartyLights, parameterName=UsePulsing}
  )
  n36 = Power(
   Base=SphereMask(
    A=NamedRerouteUsage()
    B=NamedRerouteUsage()
    Radius=LinearInterpolate(A=@n31, B=@n32, Alpha=@n35)
    Hardness=ScalarParameter() {defaultValue=0.1, group=PartyLights, parameterName=CoreMaskHardness}
   )
   Exp=ScalarParameter() {defaultValue=4, group=PartyLights, parameterName=CoreMaskExponent}
  )
  n37 = ScalarParameter() {defaultValue=25, group=PartyLights, parameterName=CoreScale}
  n38 = Multiply(A=@n30.RGB, B=NamedRerouteDeclaration(None=Multiply(A=@n36, B=@n37)) {name=CoreMask, nodeColor=<rgba 1 0.7734 0 1>})
  n39 = Multiply(A=Add(A=Multiply(A=@n20.Result, B=@n21.RGB), B=Add(A=@n29, B=@n38)), B=NamedRerouteUsage())
  n40 = ScalarParameter() {defaultValue=0.4, group=PartyLights, parameterName=PulsingLow}
  n41 = ScalarParameter() {defaultValue=1, group=PartyLights, parameterName=PulsingHigh}
  n42 = NamedRerouteUsage()
  n43 = ScalarParameter() {group=PartyLights, parameterName=TimestampA}
  n44 = ScalarParameter() {defaultValue=0.5, group=PartyLights, parameterName=TransitionLength}
  n45 = ScalarParameter() {defaultValue=1, group=PartyLights, parameterName=TimestampB}
  n46 = StaticBool()
  n47 = Multiply(
   A=Multiply(
    A=StaticSwitchParameter(True=@n6, False=@n39) {group=PartyLights, parameterName=UseCoreMaskOnly}
    B=LinearInterpolate(A=@n40, B=@n41, Alpha=@n42)
   )
   B=StaticSwitchParameter(
    True=Constant() {r=1}
    False=MaterialFunctionCall(TimestampA=@n43, TransitionLength_In=@n44, TimestampB=@n45, TransitionLength_Out=@n44, Transition In Only?=@n46) {materialFunction=&MF_Timestamp_InOut_01}.SmoothStep
   ) {group=PartyLights, parameterName=Debug_RemoveTimestamp}
  )
  n48 = Add(
   A=@n1.EmissiveColor
   B=StaticSwitchParameter(True=Multiply(A=@n47, B=@n5), False=@n47) {group=PartyLights, parameterName=UseSimpleShadingModel}
  )
  n49 = StaticSwitchParameter(
   True=StaticSwitchParameter(True=Add(A=@n1.EmissiveColor, B=@n2), False=Add(A=@n48, B=@n2)) {group=PartyLights, parameterName=UseOnlyAdditionalPulsingEmissive}
   False=@n48
  ) {group=PartyLights, parameterName=UseAdditionalPulsingEmissive}
  FunctionOutput(None=StaticSwitchParameter(
    True=SetMaterialAttributes!(MaterialAttributes=@n1.MaterialAttributes, EmissiveColor=@n49) {attributeSetTypes [#1]}
    False=@n1.MaterialAttributes
   ) {group=PartyLights, parameterName=UsePartyLights})
  n50 = NamedRerouteUsage()
  n51 = NamedRerouteUsage()
  n52 = ScalarParameter() {defaultValue=1700, group=PartyLights, parameterName=LightsMaskRadius}
  n53 = ScalarParameter() {defaultValue=0.15, group=PartyLights, parameterName=LightsMaskHardness}
  NamedRerouteDeclaration(None=Power(
    Base=SphereMask(A=@n50, B=@n51, Radius=@n52, Hardness=@n53)
    Exp=ScalarParameter() {defaultValue=12, group=PartyLights, parameterName=LightMaskExponent}
   )) {name=FinalFalloff, nodeColor=<rgba 0.6641 1 0 1>}
  n54 = Constant3Vector() {constant=<rgba 1 1 0 0>}
  NamedRerouteDeclaration(None=Sign(None=DotProduct(A=@n10.PS, B=@n54))) {name=NormalSign.XY, nodeColor=<rgba 1 0 0.9609 1>}
  n55 = TextureSampleParameter2D() {group=PartyLights, parameterName=AdditionalPulsingEmissiveMask, texture=/Engine/EditorLandscapeResources/WhiteSquareTexture}
  n56 = Multiply(
   A=Multiply(A=ChannelMaskParameter(None=@n55.RGBA) {group=PartyLights, parameterName=AdditionalPulsingEmissiveMaskChannel}.R, B=NamedRerouteUsage())
   B=VectorParameter() {group=PartyLights, parameterName=AdditionalPulsingColor, defaultValue=<rgba 1 1 1 0>}.RGB
  )
  n57 = ConstantBiasScale(None=Sine(None=Time()) {period=0.6}) {bias=1.1}
  NamedRerouteDeclaration(None=Multiply(A=Multiply(A=@n56, B=@n57), B=ScalarParameter() {group=PartyLights, parameterName=UsePulsing})) {name=AdditionalPulsing, nodeColor=<rgba 0 1 0.1094 1>}
  NamedRerouteDeclaration(None=@n1.BaseColor) {name=BaseColorIn, nodeColor=<rgba 0 1 0.1562 1>}
 ]}
